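# Default Lit front-end environment material, named as a red emissive for the Midas BP12 scene.
# An overbright red vector parameter is multiplied by an inverted Fresnel, giving an intense glow that fades toward grazing edges.
# Four constant colors (dark red, greys, near-white) sit alongside, with no output connections shown.
# Usage flags cover Nanite, skeletal meshes and mesh deformers.
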Material /Game/Environments/FrontEnd/Materials/BP12/Emissive_Red_Midas_FE
graph {
 properties {
  bUsedWithMeshDeformer=true
  bUsedWithNanite=true
  bUsedWithSkeletalMesh=true
  pixelDepthOffsetMode=PDOM_Legacy
  usedShadingModels=Default Lit
 }
 nodes [
  Constant3Vector() {constant=<rgba 0.0729 0 0 1>}
  Constant3Vector() {constant=<rgba 0.2716 0.276 0.2726 1>}
  Constant3Vector() {constant=<rgba 1 1 0.998 1>}
  Constant3Vector() {constant=<rgba 0.3594 0.3594 0.3587 1>}
  Multiply(A=VectorParameter() {defaultValue=<rgba 20 0 0 0>}.RGB, B=OneMinus(None=Fresnel() {exponent=1}))
 ]
}
