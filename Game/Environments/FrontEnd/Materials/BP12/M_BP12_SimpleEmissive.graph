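# Unlit, Nanite-compatible material for the BP12 front-end environment that produces a simple emissive color with an optional radial falloff.
# A Color parameter is scaled by Brightness, with a static Gradient switch optionally multiplying it by the falloff.
# The falloff comes from either world-space distance to the object position over Radius or the RadialGradientExponential function using Radius and Density, selected by UseWorldSpaceGradient.
# Both falloff variants are shaped by a shared Exp power.
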Material /Game/Environments/FrontEnd/Materials/BP12/M_BP12_SimpleEmissive
graph {
 properties {bUsedWithNanite=true, pixelDepthOffsetMode=PDOM_Legacy, shadingModel=MSM_Unlit, usedShadingModels=Unlit}
 nodes [
  Constant3Vector()
  n1 = ScalarParameter() {defaultValue=0.9143, parameterName=Radius}
  n2 = Divide(A=Distance(A=ObjectPositionWS(), B=WorldPosition().XYZ), B=@n1)
  n3 = ScalarParameter() {defaultValue=3, parameterName=Exp}
  n4 = ScalarParameter() {defaultValue=1, parameterName=Density}
  n5 = StaticSwitchParameter(
   True=Power(Base=OneMinus(None=@n2), Exp=@n3)
   False=Power(
    Base=MaterialFunctionCall(Radius=@n1, Density=@n4) {materialFunction=&RadialGradientExponential}.RadialGradientExponential
    Exp=@n3
   )
  ) {parameterName=UseWorldSpaceGradient}
  n6 = VectorParameter() {parameterName=Color, defaultValue=<rgba 2.9989 3 2.7481 1>}
  Multiply(
   A=StaticSwitchParameter(True=Multiply(A=@n5, B=@n6.RGB), False=@n6.RGB) {parameterName=Gradient}
   B=ScalarParameter() {defaultValue=1, parameterName=Brightness}
  )
  ScalarParameter() {defaultValue=1, parameterName=Opacity}
 ]
}
