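# Lit surface material for a Battle Pass 12 front-end floor pad, with an optional parameter-driven opening animation.
# Mask channels blend metallic, roughness and specular-style parameter pairs, and also tint the diffuse texture.
# The UseAnim switch enables a two-stage offset from Open: a short drop along negative Z, then a slide over the remaining range.
# Red and green vertex colors blend three direction parameters, so painted pieces slide different ways.
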Material /Game/Environments/FrontEnd/Materials/BP12/M_BPfloorPad
graph {
 properties {pixelDepthOffsetMode=PDOM_Legacy, usedShadingModels=Default Lit}
 nodes [
  n1 = TextureSampleParameter2D() {
   parameterName=S
   samplerType=SAMPLERTYPE_Masks
   texture=/Game/Environments/Apollo/Sets/MountainBase/Props/Textures/T_MountainBase_HoloTable_01_S
  }
  LinearInterpolate(A=ScalarParameter() {parameterName=M1}, B=ScalarParameter() {defaultValue=0.8857, parameterName=M2}, Alpha=@n1.G)
  LinearInterpolate(A=ScalarParameter() {parameterName=R1}, B=ScalarParameter() {defaultValue=0.5429, parameterName=R2}, Alpha=@n1.G)
  LinearInterpolate(A=ScalarParameter() {defaultValue=0.5, parameterName=S1}, B=ScalarParameter() {defaultValue=0.5, parameterName=S2}, Alpha=@n1.R)
  Multiply(
   A=TextureSampleParameter2D() {parameterName=D, texture=/Game/Environments/Apollo/Sets/MountainBase/Props/Textures/T_MountainBase_HoloTable_02_D}.RGB
   B=LinearInterpolate(A=ScalarParameter() {defaultValue=1, parameterName=B_metal1}, B=ScalarParameter() {defaultValue=0.2952, parameterName=B_metal2}, Alpha=@n1.G)
  )
  n2 = Multiply(A=ScalarParameter() {defaultValue=12.0961, parameterName=MoveDown}, B=Constant3Vector() {constant=<rgba 0 0 -1 0>})
  n3 = Saturate(None=ScalarParameter() {parameterName=Open})
  n4 = ScalarParameter() {defaultValue=16, parameterName=FractionforMovedown}
  n5 = Saturate(None=Multiply(A=@n3, B=@n4))
  n6 = VertexColor()
  n7 = LinearInterpolate(
   A=VectorParameter() {parameterName=Dir1, defaultValue=<rgba 1 0.6857 0 1>}.RGB
   B=VectorParameter() {parameterName=Dir2, defaultValue=<rgba -0.3048 -6.9433 0 1>}.RGB
   Alpha=@n6.R
  )
  n8 = VectorParameter() {parameterName=Dir3, defaultValue=<rgba -9.5271 6.5864 0 1>}
  n9 = Multiply(
   A=ScalarParameter() {defaultValue=120.3828, parameterName=MoveDistance}
   B=Normalize(VectorInput=LinearInterpolate(A=@n7, B=@n8.RGB, Alpha=@n6.G))
  )
  n10 = Constant() {r=1}
  n11 = Divide(A=Constant() {r=1}, B=@n4)
  n12 = Subtract(A=Constant() {r=1}, B=@n11)
  n13 = Saturate(None=Multiply(A=Divide(A=@n10, B=@n12), B=Subtract(A=@n3, B=@n11)))
  StaticSwitchParameter(True=Add(A=Multiply(A=@n2, B=@n5), B=Multiply(A=@n9, B=@n13)), False=Constant3Vector()) {parameterName=UseAnim}
  TextureSampleParameter2D() {
   parameterName=N
   samplerType=SAMPLERTYPE_Normal
   texture=/Game/Environments/Apollo/Sets/MountainBase/Props/Textures/T_MountainBase_HoloTable_01_N
  }
 ]
}
