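# Translucent, per-pixel lit glass material for BP12 front-end key art, reusing Hamster Ball crack and dome normal textures.
# GlassCrack drives FlattenNormal on two layers of the crack normal: full-scale cracks strengthen throughout, while a 5x-tiled, offset copy appears above 0.5.
# The layers' XY components are summed with Z rebuilt as 1.
# A thickness-scaled Fresnel blends between inner and outer transparency parameters.
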Material /Game/Environments/FrontEnd/Materials/BP12/M_CT_KeyArt_Glass
graph {
 properties {
  bCastRayTracedShadows=false
  bUsedWithMeshDeformer=true
  bUsedWithNiagaraMeshParticles=true
  bUsedWithSkeletalMesh=true
  blendMode=BLEND_Translucent
  pixelDepthOffsetMode=PDOM_Legacy
  translucencyLightingMode=TLM_SurfacePerPixelLighting
  usedShadingModels=Default Lit
 }
 nodes [
  n1 = TextureCoordinate() {uTiling=5, vTiling=5}
  n2 = Constant2Vector() {g=0.65, r=0.25}
  n3 = Multiply(
   A=TextureSample(UVs=Add(A=@n1, B=@n2)) {samplerType=SAMPLERTYPE_Normal, texture=/Game/Athena/DrivableVehicles/Hamster_Ball/Textures/T_Hamster_Ball_Glass_Cracks}.RGB
   B=Constant3Vector() {constant=<rgba 1 1 0 0>}
  )
  n4 = OneMinus(None=ScalarParameter() {parameterName=GlassCrack})
  n5 = Multiply(A=Clamp(None=@n4) {maxDefault=0.5}) {constB=2}
  n6 = Multiply(
   A=TextureSample(UVs=TextureCoordinate()) {samplerType=SAMPLERTYPE_Normal, texture=&T_Hamster_Ball_Glass_Cracks}.RGB
   B=Constant3Vector() {constant=<rgba 1 1 0 0>}
  )
  n7 = Clamp(None=@n4)
  Add(
   A=Constant3Vector() {constant=<rgba 0 0 1 0>}
   B=Add(
    A=MaterialFunctionCall(Normal=@n3, Flatness=@n5) {materialFunction=&FlattenNormal}.Result
    B=MaterialFunctionCall(Normal=@n6, Flatness=@n7) {materialFunction=&FlattenNormal}.Result
   )
  )
  Constant()
  Constant() {r=0.1}
  Constant() {r=1}
  Constant() {r=0.02}
  n8 = Fresnel(ExponentIn=ScalarParameter() {defaultValue=1, parameterName=Fresnel Exponent})
  n9 = ScalarParameter() {defaultValue=1, parameterName=Glass thickness}
  LinearInterpolate(
   A=ScalarParameter() {defaultValue=0.1, parameterName=Fresnel Inner Transparency}
   B=ScalarParameter() {defaultValue=0.85, parameterName=Fresnel Outer Transparency}
   Alpha=Clamp(None=Multiply(A=@n8, B=@n9))
  )
  Multiply(A=Add(A=@n4) {constB=-0.5}) {constB=2}
  TextureSample() {samplerType=SAMPLERTYPE_Normal, texture=/Game/Athena/DrivableVehicles/Hamster_Ball/Textures/T_Hamster_Dome_N}
 ]
}
