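# M_CT_KeyArt_Glass_AG_FE is a translucent, lit glass material for front-end key art, usable on skeletal meshes.
# A Fresnel term with adjustable exponent is scaled by glass thickness, clamped, and blends between inner and outer transparency values.
# Scalar parameters expose color, metallic, specular and roughness, alongside a constant (0,0,1) vector.
# It uses per-pixel surface translucency lighting and disables ray-traced shadow casting.
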Material /Game/Environments/FrontEnd/Materials/BP12/M_CT_KeyArt_Glass_AG_FE
graph {
 properties {
  bCastRayTracedShadows=false
  bUsedWithMeshDeformer=true
  bUsedWithSkeletalMesh=true
  blendMode=BLEND_Translucent
  pixelDepthOffsetMode=PDOM_Legacy
  translucencyLightingMode=TLM_SurfacePerPixelLighting
  usedShadingModels=Default Lit
 }
 nodes [
  Constant3Vector() {constant=<rgba 0 0 1 0>}
  n1 = Fresnel(ExponentIn=ScalarParameter() {defaultValue=1, parameterName=Fresnel Exponent})
  n2 = ScalarParameter() {defaultValue=1, parameterName=Glass thickness}
  LinearInterpolate(
   A=ScalarParameter() {defaultValue=0.2, parameterName=Fresnel Inner Transparency}
   B=ScalarParameter() {defaultValue=0.85, parameterName=Fresnel Outer Transparency}
   Alpha=Clamp(None=Multiply(A=@n1, B=@n2))
  )
  ScalarParameter() {defaultValue=0.5, parameterName=Metal}
  ScalarParameter() {defaultValue=1, parameterName=Spec}
  ScalarParameter() {defaultValue=0.02, parameterName=Rough}
  ScalarParameter() {parameterName=Color}
 ]
}
