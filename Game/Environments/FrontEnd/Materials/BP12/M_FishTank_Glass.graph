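# Translucent, two-sided lit material for front-end fish tank glass that fakes the water inside.
# An animated sine wave offsets object-relative height to form a moving waterline mask.
# That mask gates panning bubbles and, faded by inverse Fresnel, blends water tint and opacity over the glass.
# Tint and bubbles are scaled as emissive, then added to a fake window reflection sampled by reflection vector.
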Material /Game/Environments/FrontEnd/Materials/BP12/M_FishTank_Glass
graph {
 properties {
  blendMode=BLEND_Translucent
  pixelDepthOffsetMode=PDOM_Legacy
  translucencyLightingMode=TLM_SurfacePerPixelLighting
  twoSided=true
  usedShadingModels=Default Lit
 }
 nodes [
  n1 = ReflectionVectorWS()
  n2 = TextureCoordinate()
  n3 = ScalarParameter() {defaultValue=1.1831, parameterName=Bubblesscale}
  n4 = Constant()
  n5 = ScalarParameter() {defaultValue=0.11, parameterName=Pan}
  n6 = TextureSample(UVs=Panner(Coordinate=Multiply(A=@n2, B=@n3), Speed=AppendVector(A=@n4, B=@n5))) {samplerType=SAMPLERTYPE_Masks, texture=/Game/Weapons/WeaponSkins/Wraps/Textures/SlurpMonster/T_Wrap_SlurpMonster_Bubbles_MB}
  n7 = ScalarParameter() {defaultValue=9.6834, parameterName=C}
  n8 = Multiply(A=Time(), B=ScalarParameter() {defaultValue=25, parameterName=Speed})
  n9 = WorldPosition()
  n10 = Frac(None=Divide(A=Add(A=@n8, B=@n9.XYZ), B=ScalarParameter() {defaultValue=50.3294, parameterName=Wavelength}))
  n11 = Add(A=Sine(None=ComponentMask(None=@n10) {r=true}), B=Sine(None=ComponentMask(None=@n10) {g=true}))
  n12 = ScalarParameter() {defaultValue=0.7981, parameterName=Amplitude}
  n13 = ObjectPositionWS()
  n14 = ComponentMask(None=Add(A=Multiply(A=@n11, B=@n12), B=Subtract(A=@n9.XYZ, B=@n13))) {b=true}
  n15 = ScalarParameter() {defaultValue=16.6687, parameterName=Offset}
  n16 = Saturate(None=Divide(A=Subtract(A=@n14, B=@n15), B=ScalarParameter() {defaultValue=-0.7173, parameterName=Blend}))
  n17 = Multiply(A=Multiply(A=VectorParameter() {parameterName=Bubbles, defaultValue=<rgba 3 3 3 1>}.RGB, B=Power(Base=@n6.R, Exp=@n7)), B=@n16)
  n18 = Fresnel(ExponentIn=ScalarParameter() {defaultValue=6.7981, parameterName=exp}, Normal=PixelNormalWS()) {baseReflectFraction=0}
  n19 = Multiply(A=@n16, B=OneMinus(None=Saturate(None=@n18)))
  n20 = Add(
   A=VectorParameter() {parameterName=ColorGlass, defaultValue=<rgba 0.0509 0.1198 0.0834 1>}.RGB
   B=LinearInterpolate(A=Constant3Vector(), B=VectorParameter() {parameterName=ColorWater, defaultValue=<rgba 0 0.0466 0.0625 1>}.RGB, Alpha=@n19)
  )
  Add(
   A=Multiply(
    A=TextureSample(UVs=@n1) {samplerType=SAMPLERTYPE_LinearColor, texture=/Game/Environments/Sets/Neo/Architecture/Textures/T_FakeWindowReflections}.R
    B=ScalarParameter() {defaultValue=0.6602, parameterName=Reflection}
   )
   B=Multiply(A=Add(A=@n17, B=@n20), B=ScalarParameter() {defaultValue=0.4465, parameterName=Emissive})
  )
  LinearInterpolate(
   A=ScalarParameter() {defaultValue=0.1331, parameterName=OpacityGlas}
   B=ScalarParameter() {defaultValue=0.3809, parameterName=Opacitywater}
   Alpha=@n19
  )
  ScalarParameter() {parameterName=R}
  ScalarParameter() {defaultValue=1, parameterName=S}
  ScalarParameter() {defaultValue=1.006, parameterName=Refaraction}
  TextureSample() {samplerType=SAMPLERTYPE_Normal, texture=/Game/Athena/Environments/Island/CubeGrowth/T_CubeGrowth_normals}
 ]
}
