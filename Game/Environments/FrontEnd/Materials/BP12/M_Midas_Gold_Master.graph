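# Master material for the Midas gold look in the BP12 front-end environment, using Default Lit shading.
# It holds a fixed gold color, a constant of one, a roughness scalar parameter and a normal map texture parameter defaulting to a Princess Castle banquet table setting texture.
# Usage flags cover skeletal meshes, mesh deformers, and Niagara mesh particles and sprites.
# Only the roughness and normal texture are exposed as parameters.
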Material /Game/Environments/FrontEnd/Materials/BP12/M_Midas_Gold_Master
graph {
 properties {
  bUsedWithMeshDeformer=true
  bUsedWithNiagaraMeshParticles=true
  bUsedWithNiagaraSprites=true
  bUsedWithSkeletalMesh=true
  pixelDepthOffsetMode=PDOM_Legacy
  usedShadingModels=Default Lit
 }
 nodes [
  Constant3Vector() {constant=<rgba 0.6927 0.4472 0.1313 1>}
  Constant() {r=1}
  ScalarParameter() {defaultValue=0.325, parameterName=Shine Roughnmess}
  TextureSampleParameter2D() {
   parameterName=Texture
   samplerType=SAMPLERTYPE_Normal
   texture=/Game/Environments/Sets/PrincessCastle/Props/Textures/T_PrincessCastle_Banquet_Table_Setting_N
  }
 ]
}
