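# Unlit front-end monitor material (BP12) that shows a reward image as a glitchy screen portrait over a two-color radial gradient.
# The rotated, scaled image is sampled three times, with sine-driven opposite offsets on red and blue for animated chromatic aberration.
# Summed alpha, cropped by four edge masks, loses panning sine scanlines and periodic noise streaks to form the blend opacity.
# UsePortrait fades between plain gradient and composite.
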Material /Game/Environments/FrontEnd/Materials/BP12/M_Monitor_BP12
graph {
 properties {numCustomizedUVs=1, pixelDepthOffsetMode=PDOM_Legacy, shadingModel=MSM_Unlit, usedShadingModels=Unlit}
 nodes [
  n1 = LinearInterpolate(
   A=VectorParameter() {parameterName=Color1, defaultValue=<rgba 0.235 0.0357 0.0176 1>}.RGB
   B=VectorParameter() {parameterName=Color2, defaultValue=<rgba 3 0.833 0.045 1>}.RGB
   Alpha=SphereMask(
    A=TextureCoordinate()
    B=Constant2Vector() {g=0.5, r=0.5}
    Radius=ScalarParameter() {defaultValue=0.7238, group=Panel, parameterName=Radius}
    Hardness=ScalarParameter() {defaultValue=-3.4962, group=Panel, parameterName=Falloff}
   )
  )
  n2 = Constant() {r=0.25}
  n3 = TextureCoordinate()
  n4 = ScalarParameter() {defaultValue=1.2, group=Portrait, parameterName=PortraitScale_U}
  n5 = ScalarParameter() {defaultValue=0.6, group=Portrait, parameterName=PortraitScale_V}
  n6 = Add(
   A=MaterialFunctionCall(
    UVs=MaterialFunctionCall!(Rotation_Fraction=@n2, TexCoord=@n3) {materialFunction=/FortUILibrary/Foundation/Art_ToolKit/AngledBoxes/MF_UI_Rotator_Fractions}.Result
    Texture Scale=AppendVector(A=@n4, B=@n5)
   ) {materialFunction=&ScaleUVsByCenter}.UVs
   B=AppendVector(
    A=ScalarParameter() {group=Portrait, parameterName=Translate_Portrait_U}
    B=ScalarParameter() {defaultValue=0.15, group=Portrait, parameterName=Translate_Portrait_V}
   )
  )
  n7 = Multiply(A=Time(), B=ScalarParameter() {defaultValue=1, group=Global, parameterName=GlobalTimeControl})
  n8 = Multiply(A=@n7, B=ScalarParameter() {defaultValue=0.5, group=Portrait, parameterName=RedChannel_Time_Sine_1})
  n9 = LinearInterpolate(Alpha=Add(A=Sine(None=@n8)))
  n10 = Multiply(A=@n7, B=ScalarParameter() {defaultValue=4, group=Portrait, parameterName=RedChannel_Time_Sine_2})
  n11 = LinearInterpolate(Alpha=Add(A=Sine(None=@n10)))
  n12 = Multiply(A=@n7, B=ScalarParameter() {defaultValue=1, group=Portrait, parameterName=RedChannel_Time_Sine_3})
  n13 = LinearInterpolate(Alpha=Add(A=Sine(None=@n12)))
  n14 = LinearInterpolate(
   A=ScalarParameter() {defaultValue=0.009, group=Portrait, parameterName=RedChannel_Min_Offset}
   B=ScalarParameter() {defaultValue=0.01, group=Portrait, parameterName=RedChannel_Max_Offset}
   Alpha=DotProduct(A=DotProduct(A=@n9, B=@n11), B=@n13)
  )
  n15 = Constant()
  n16 = TextureSampleParameter2D(UVs=Add(A=@n6, B=AppendVector(A=@n14, B=@n15))) {parameterName=RewardImage, samplerSource=SSM_Clamp_WorldGroupSettings, texture=/Game/UI/Foundation/Textures/Icons/Items/T-FNBR-BattlePass-L}
  n17 = TextureSampleParameter2D(UVs=@n6) {parameterName=RewardImage, samplerSource=SSM_Clamp_WorldGroupSettings, texture=&T-FNBR-BattlePass-L}
  n18 = Sine(None=Multiply(A=@n7, B=ScalarParameter() {defaultValue=0.25, group=Portrait, parameterName=BlueChannel_Time_Sine_1}))
  n19 = Sine(None=Multiply(A=@n7, B=ScalarParameter() {defaultValue=6, group=Portrait, parameterName=BlueChannel_Time_Sine_2}))
  n20 = DotProduct(A=LinearInterpolate(Alpha=Add(A=@n18)), B=LinearInterpolate(Alpha=Add(A=@n19)))
  n21 = Multiply(A=@n7, B=ScalarParameter() {defaultValue=0.75, group=Portrait, parameterName=BlueChannel_Time_Sine_3})
  n22 = LinearInterpolate(Alpha=Add(A=Sine(None=@n21)))
  n23 = Multiply(A=LinearInterpolate(
    A=ScalarParameter() {defaultValue=0.009, group=Portrait, parameterName=BlueChannel_Min_Offset}
    B=ScalarParameter() {defaultValue=0.01, group=Portrait, parameterName=BlueChannel_Max_Offset}
    Alpha=DotProduct(A=@n20, B=@n22)
   )) {constB=-1}
  n24 = Constant()
  n25 = TextureSampleParameter2D(UVs=Add(A=@n6, B=AppendVector(A=@n23, B=@n24))) {parameterName=RewardImage, samplerSource=SSM_Clamp_WorldGroupSettings, texture=&T-FNBR-BattlePass-L}
  n26 = MaterialFunctionCall(X=@n16.R, Y=@n17.G, Z=@n25.B) {materialFunction=&MakeFloat3}
  n27 = Add(A=Add(A=@n16.A, B=@n17.A), B=@n25.A)
  n28 = ScalarParameter() {defaultValue=-0.05, group=Portrait, parameterName=Top_Mask_Offset}
  n29 = TextureCoordinate()
  n30 = Multiply(A=ComponentMask(None=Add(A=@n29, B=@n28)) {r=true}, B=ScalarParameter() {defaultValue=16, group=Portrait, parameterName=Mask_Falloff_Top})
  n31 = ScalarParameter() {group=Portrait, parameterName=Right_Mask_Offset}
  n32 = Multiply(A=ComponentMask(None=Add(A=@n29, B=@n31)) {g=true}, B=ScalarParameter() {defaultValue=128, group=Portrait, parameterName=Mask_Falloff_Right})
  n33 = OneMinus(None=@n29)
  n34 = ScalarParameter() {defaultValue=-0.3, group=Portrait, parameterName=Bottom_Mask_Offset}
  n35 = Multiply(
   A=ComponentMask(None=Add(A=@n33, B=@n34)) {r=true}
   B=ScalarParameter() {defaultValue=1.5, group=Portrait, parameterName=Mask_Falloff_Bottom}
  )
  n36 = ScalarParameter() {group=Portrait, parameterName=Left_Mask_Offset}
  n37 = Multiply(A=ComponentMask(None=Add(A=@n33, B=@n36)) {g=true}, B=ScalarParameter() {defaultValue=128, group=Portrait, parameterName=Mask_Falloff_Left})
  n38 = DotProduct(A=DotProduct(A=Saturate(None=@n30), B=Saturate(None=@n32)), B=DotProduct(A=Saturate(None=@n35), B=Saturate(None=@n37)))
  n39 = TextureCoordinate()
  n40 = Time()
  n41 = Multiply(A=Panner(Coordinate=@n39, Time=@n40) {speedX=-0.0125}, B=ScalarParameter() {defaultValue=8, group=Portrait, parameterName=Lines_1_Amount})
  n42 = Multiply(A=Panner(Coordinate=@n39, Time=@n40) {speedX=0.0063}, B=ScalarParameter() {defaultValue=16, group=Portrait, parameterName=Lines_2_Amount})
  n43 = DotProduct(A=ComponentMask(None=Sine(None=@n41)) {r=true}, B=ComponentMask(None=Sine(None=@n42)) {r=true})
  n44 = Panner(Time=@n40) {speedX=0.1}
  n45 = ScalarParameter() {defaultValue=2, group=Portrait, parameterName=Lines_3_Amount}
  n46 = ComponentMask(None=Sine(None=Multiply(A=@n44, B=@n45))) {r=true}
  n47 = Multiply(A=Saturate(None=DotProduct(A=@n43, B=@n46)), B=ScalarParameter() {defaultValue=16, group=Portrait, parameterName=ScanLines_Softness})
  n48 = Subtract(
   A=Saturate(None=DotProduct(A=@n27, B=@n38))
   B=Multiply(A=Saturate(None=@n47), B=ScalarParameter() {defaultValue=0.075, group=Portrait, parameterName=ScanLines_Amount})
  )
  n49 = Multiply(A=Time()) {constB=0.125}
  n50 = Constant() {r=1}
  n51 = Saturate(None=LinearInterpolate(Alpha=Fmod(A=@n49, B=@n50)) {constA=-10, constB=10})
  n52 = TextureCoordinate() {uTiling=0.5, vTiling=0.5}
  n53 = Power(Base=TextureSample(UVs=Panner(Coordinate=@n52) {speedX=-16}) {samplerType=SAMPLERTYPE_LinearColor, texture=/Game/Global/Textures/Default/Noise/T_Wrap_SnowStreak_Noise}.R)
  n54 = Power(Base=TextureSample(UVs=Panner(Coordinate=@n52)) {samplerType=SAMPLERTYPE_LinearColor, texture=&T_Wrap_SnowStreak_Noise}.G)
  n55 = TextureSample(UVs=Panner(Coordinate=@n52) {speedX=16}) {samplerType=SAMPLERTYPE_LinearColor, texture=&T_Wrap_SnowStreak_Noise}
  n56 = Add(A=Add(A=DotProduct(A=@n9, B=@n53), B=DotProduct(A=@n11, B=@n54)), B=DotProduct(A=@n13, B=Power(Base=@n55.B)))
  n57 = Multiply(A=Power(Base=DotProduct(A=@n51, B=@n56)), B=ScalarParameter() {defaultValue=0.0125, group=Portrait, parameterName=SquaresAmount})
  n58 = Saturate(None=Subtract(A=Saturate(None=@n48), B=Saturate(None=@n57)))
  n59 = ScalarParameter() {defaultValue=0.666, group=Portrait, parameterName=PortraitAlpha}
  LinearInterpolate(
   A=@n1
   B=LinearInterpolate(A=@n1, B=Saturate(None=@n26.Result), Alpha=Multiply(A=@n58, B=@n59))
   Alpha=ScalarParameter() {defaultValue=1, group=Global, parameterName=UsePortrait}
  )
 ]
}
