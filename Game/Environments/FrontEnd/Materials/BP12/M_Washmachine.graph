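# Washing machine material for a front-end (BP12) environment, with static switches for glass, clock, underwear and vibrate-only variants.
# The glass branch draws emissive, translucency-driven water with panning bubbles and a sine-wave level; the other uses standard diffuse, spec and normal maps.
# World Position Offset combines pivot spin, hinged door rotation and height-scaled sine vibration, using vertex color channels as masks.
# Outside vibrate-only mode, normals follow the pivot rotation.
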Material /Game/Environments/FrontEnd/Materials/BP12/M_Washmachine
graph {
 properties {bUseMaterialAttributes=true, pixelDepthOffsetMode=PDOM_Legacy, usedShadingModels=Default Lit}
 nodes [
  n1 = VectorParameter() {parameterName=Glass, defaultValue=<rgba 0 0 0 1>}
  n2 = VectorParameter() {parameterName=Bubbles, defaultValue=<rgba 0.3 0.3 0.3 1>}
  n3 = Multiply(A=TextureCoordinate(), B=ScalarParameter() {defaultValue=5.4891, parameterName=Bubblesscale})
  n4 = AppendVector(A=Constant(), B=ScalarParameter() {defaultValue=-0.1, parameterName=Pan})
  n5 = Power(
   Base=TextureSample(UVs=Panner(Coordinate=@n3, Speed=@n4)) {samplerType=SAMPLERTYPE_Masks, texture=/Game/Weapons/WeaponSkins/Wraps/Textures/SlurpMonster/T_Wrap_SlurpMonster_Bubbles_MB}.R
   Exp=ScalarParameter() {defaultValue=1, parameterName=C}
  )
  n6 = VectorParameter() {parameterName=ColorWater, defaultValue=<rgba 0 0.0466 0.0625 1>}
  n7 = Multiply(A=Time(), B=ScalarParameter() {defaultValue=0.7524, parameterName=Speedwave})
  n8 = WorldPosition()
  n9 = Frac(None=Divide(A=Add(A=@n7, B=@n8.XYZ), B=ScalarParameter() {defaultValue=50.3294, parameterName=Wavelength}))
  n10 = Add(A=Sine(None=ComponentMask(None=@n9) {r=true}), B=Sine(None=ComponentMask(None=@n9) {g=true}))
  n11 = ScalarParameter() {defaultValue=0.7981, parameterName=Amplitude}
  n12 = ObjectPositionWS()
  n13 = ComponentMask(None=Add(A=Multiply(A=@n10, B=@n11), B=Subtract(A=@n8.XYZ, B=@n12))) {b=true}
  n14 = ScalarParameter() {defaultValue=16.6687, parameterName=Offset}
  n15 = Saturate(None=Divide(A=Subtract(A=@n13, B=@n14), B=ScalarParameter() {defaultValue=-0.7173, parameterName=Blend}))
  n16 = ScalarParameter() {defaultValue=1, parameterName=Enabled}
  n17 = LinearInterpolate(A=@n1.RGB, B=Add(A=Multiply(A=@n2.RGB, B=@n5), B=LinearInterpolate(A=@n1.RGB, B=@n6.RGB, Alpha=@n15)), Alpha=@n16)
  n18 = ScalarParameter() {defaultValue=0.2, parameterName=Opacity2}
  n19 = LinearInterpolate(A=@n18, B=LinearInterpolate(A=@n18, B=ScalarParameter() {defaultValue=0.2, parameterName=Opacity}, Alpha=@n15), Alpha=@n16)
  n20 = TextureSampleParameter2D() {parameterName=Diffuse, texture=/Game/Environments/Props/Residential/Textures/Prop_ClothesWasher_02_D}
  n21 = TextureSampleParameter2D() {parameterName=Spec, samplerType=SAMPLERTYPE_Masks, texture=/Game/Environments/Props/Residential/Textures/Prop_ClothesWasher_02_S}
  n22 = Multiply(A=ScalarParameter() {defaultValue=1, parameterName=Roughness}, B=@n21.B)
  n23 = TextureSampleParameter2D() {parameterName=Normal, samplerType=SAMPLERTYPE_Normal, texture=/Game/Environments/Props/Residential/Textures/Prop_ClothesWasher_02_N}
  n24 = Normalize(VectorInput=Transform(None=VectorParameter() {parameterName=Axis, defaultValue=<rgba 0 1 0 1>}.RGB) {transformSourceType=TRANSFORMSOURCE_Local})
  n25 = ScalarParameter() {parameterName=MaxRotation}
  n26 = Time()
  n27 = Frac(None=Multiply(A=@n26, B=ScalarParameter() {defaultValue=0.7524, parameterName=Speed}))
  n28 = Add(A=ScalarParameter() {parameterName=MaxRotationBias}, B=Sine(None=@n27))
  n29 = StaticSwitchParameter(True=Multiply(A=Multiply(A=@n25, B=@n28), B=VertexColor().R), False=@n27) {parameterName=IsClock}
  n30 = LinearInterpolate(A=Round(None=@n29), B=@n29, Alpha=@n16)
  n31 = VertexColor()
  n32 = Subtract(A=@n31.R, B=@n31.R)
  n33 = StaticSwitchParameter(True=@n30, False=Multiply(A=@n30, B=@n32)) {parameterName=Underwear}
  n34 = StaticSwitchParameter(
   True=@n23.RGB
   False=MaterialFunctionCall(
    BaseNormal=MaterialFunctionCall(Rotation Axis=@n24, Rotation Angle=@n33) {materialFunction=&FixRotateAboutAxisNormals}.New Tangent Space Vertex Normal
    AdditionalNormal=@n23.RGB
   ) {materialFunction=&BlendAngleCorrectedNormals}.Result
  ) {parameterName=VibrateOnly}
  n35 = RotateAboutAxis(
   NormalizedRotationAxis=@n24
   RotationAngle=@n29
   PivotPoint=TransformPosition(None=VectorParameter() {parameterName=Pivot, defaultValue=<rgba 2.927 0 61 1>}.RGB) {transformType=TRANSFORMPOSSOURCE_World}
   Position=WorldPosition().XYZ
  )
  n36 = Time()
  n37 = ScalarParameter() {defaultValue=3.3053, parameterName=VibrationSpeed}
  n38 = ScalarParameter() {defaultValue=4.2968, parameterName=VibrationSpeed2}
  n39 = AppendVector(A=Sine(None=Multiply(A=@n36, B=@n37)), B=Sine(None=Multiply(A=@n36, B=@n38)))
  n40 = Sine(None=Multiply(B=ScalarParameter() {defaultValue=3.1843, parameterName=VibrationSpeed3}))
  n41 = ScalarParameter() {defaultValue=2, parameterName=VibrationDistamce}
  n42 = WorldPosition()
  n43 = Divide(
   A=Subtract(A=ComponentMask(None=@n42.XYZ) {b=true}, B=ScalarParameter() {parameterName=OffestBlend})
   B=ScalarParameter() {defaultValue=323.4204, parameterName=BlendDistance}
  )
  n44 = Time()
  n45 = Multiply(A=@n44, B=ScalarParameter() {defaultValue=3.3053, parameterName=VibrationSpeed})
  n46 = MaterialFunctionCall(Float3=ObjectPositionWS()) {materialFunction=&BreakOutFloat3Components}
  n47 = Multiply(A=@n44, B=ScalarParameter() {defaultValue=4.2968, parameterName=VibrationSpeed2})
  n48 = AppendVector(A=Sine(None=Add(A=@n45, B=@n46.R)), B=Sine(None=Add(A=@n47, B=@n46.G)))
  n49 = ScalarParameter() {defaultValue=3.1843, parameterName=VibrationSpeed3}
  n50 = Sine(None=Add(A=Multiply(B=@n49), B=@n46.B))
  n51 = Add(
   A=Multiply(A=AppendVector(A=@n39, B=@n40), B=Multiply(A=@n41, B=@n43))
   B=Multiply(A=AppendVector(A=@n48, B=@n50), B=ScalarParameter() {defaultValue=0.3, parameterName=VibrationDistamce2})
  )
  n52 = VertexColor()
  n53 = StaticSwitchParameter(True=Constant3Vector(), False=@n35) {parameterName=IsGlass}
  n54 = VectorParameter() {parameterName=AxisDoor, defaultValue=<rgba 0 0 1 1>}
  n55 = Frac(None=Multiply(A=@n26, B=ScalarParameter() {defaultValue=1, parameterName=Speed2}))
  n56 = Round(None=@n55)
  n57 = Multiply(A=ScalarParameter() {defaultValue=0.3714, parameterName=OpenAngle}, B=VertexColor().G)
  n58 = ScalarParameter() {parameterName=Open}
  n59 = WorldPosition()
  n60 = RotateAboutAxis(
   NormalizedRotationAxis=Normalize(VectorInput=Transform(None=@n54.RGB) {transformSourceType=TRANSFORMSOURCE_Local})
   RotationAngle=StaticSwitchParameter(True=LinearInterpolate(A=@n56, B=@n55, Alpha=@n16), False=LinearInterpolate(B=@n57, Alpha=@n58)) {parameterName=Underwear}
   PivotPoint=TransformPosition(None=VectorParameter() {parameterName=Pivothinge, defaultValue=<rgba -19.9655 42.9548 0 1>}.RGB) {transformType=TRANSFORMPOSSOURCE_World}
   Position=StaticSwitchParameter(True=Add(A=@n53, B=@n59.XYZ), False=@n59.XYZ) {parameterName=Underwear}
  )
  n61 = Add(A=@n53, B=@n60)
  n62 = Add(A=Multiply(A=@n53, B=@n32), B=@n60)
  n63 = StaticSwitchParameter(
   True=StaticSwitchParameter(True=Multiply(A=@n51, B=@n52.B), False=@n51) {parameterName=IsClock}
   False=Add(A=StaticSwitchParameter(True=@n61, False=@n62) {parameterName=Underwear}, B=@n51)
  ) {parameterName=VibrateOnly}
  SetMaterialAttributes!(
   MaterialAttributes=StaticSwitchParameter(
    True=MakeMaterialAttributes!(
     BaseColor
     Metallic
     Specular
     Roughness
     Anisotropy
     EmissiveColor=@n17
     Opacity=@n19
     OpacityMask
     Normal
     Tangent
     WorldPositionOffset
     SubsurfaceColor
     ClearCoat
     ClearCoatRoughness
     AmbientOcclusion
     Refraction
     CustomizedUVs_0
     CustomizedUVs_1
     CustomizedUVs_2
     CustomizedUVs_3
     CustomizedUVs_4
     CustomizedUVs_5
     CustomizedUVs_6
     CustomizedUVs_7
     PixelDepthOffset
     ShadingModel
     Displacement
    )
    False=MakeMaterialAttributes(BaseColor=@n20.RGB, Metallic=@n21.G, Specular=@n21.G, Roughness=@n22, Normal=@n34)
   ) {parameterName=IsGlass}
   World Position Offset=StaticSwitchParameter(True=Add(A=@n35, B=@n63), False=@n63) {parameterName=IsClock}
  ) {attributeSetTypes [#1]}
 ]
}
