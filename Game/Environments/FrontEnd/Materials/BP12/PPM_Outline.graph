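# Front-end post-process material that outlines custom-depth objects and color grades the screen, with an optional UI shadow.
# Outlines compare custom and scene depth across four neighbor samples, suppressed near stencil ID 20.
# Desktop shading optionally grades by world distance; mobile shading applies desaturation, contrast, vignette and two positional fake lights.
# ES2/ES3.1 feature levels replace the outline with an animated, depth-masked shimmer and tint unless Neutral is set.
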Material /Game/Environments/FrontEnd/Materials/BP12/PPM_Outline
graph {
 properties {
  blendableLocation=BL_SceneColorAfterDOF
  materialDomain=MD_PostProcess
  pixelDepthOffsetMode=PDOM_Legacy
  shadingModel=MSM_Unlit
  usedShadingModels=Unlit
 }
 nodes [
  n1 = Time()
  n2 = SceneTexture() {sceneTextureId=PPI_CustomStencil}
  Add(A=@n1, B=Multiply(A=ComponentMask(None=@n2.Color) {r=true}, B=ScalarParameter() {defaultValue=0.1198, parameterName=Random}))
  n3 = AppendVector(A=Constant() {r=1}, B=ScalarParameter() {defaultValue=1, group=UITopShadow, parameterName=UIShadowTopScale})
  n4 = ScreenPosition()
  n5 = VectorParameter() {group=UITopShadow, parameterName=UIShadowPos, defaultValue=<rgba 0.5 0 0 1>}
  n6 = ScalarParameter() {defaultValue=1, group=UITopShadow, parameterName=UIShadowTopOffset}
  n7 = Add(A=Distance(A=Multiply(A=@n3, B=@n4.ViewportUV), B=ComponentMask(None=@n5.RGB) {g=true, r=true}), B=@n6)
  n8 = ScalarParameter() {defaultValue=1, group=UITopShadow, parameterName=UIShadowTopDistance}
  n9 = Saturate(None=Power(Base=Divide(A=@n7, B=@n8), Exp=ScalarParameter() {defaultValue=3, group=UITopShadow, parameterName=UIShadowTopFalloff}))
  Add(A=@n9, B=@n6)
  n10 = ScalarParameter() {defaultValue=1, group=UITopShadow, parameterName=UIShadowTopBrightness}
  Add(A=@n9, B=@n10)
  n11 = ScreenPosition()
  n12 = Panner(
   Coordinate=Multiply(A=@n11.ViewportUV, B=ScalarParameter() {defaultValue=1, parameterName=TilingSineMask})
   Speed=AppendVector(A=ScalarParameter() {defaultValue=0.1, parameterName=Pansine}, B=Constant())
  )
  ComponentMask(None=@n12) {g=true, r=true}
  Constant3Vector()
  n13 = ComponentMask(None=SceneTexture() {sceneTextureId=PPI_PostProcessInput0}.Color) {b=true, g=true, r=true}
  LinearInterpolate(
   A=@n13
   B=LinearInterpolate(A=@n13, Alpha=ScalarParameter() {defaultValue=0.1, parameterName=FakelgihtonPC})
   Alpha=CollectionParameter() {collection=/Game/Environments/FrontEnd/Materials/Cozy/MPC_Cozy, parameterName=isLowDetail}
  )
  n14 = Distance(A=WorldPosition().XYZ, B=VectorParameter() {group=PC_CC, parameterName=PC_CC_Location, defaultValue=<rgba 0 0 0 1>}.RGB)
  n15 = ScalarParameter() {defaultValue=500, group=PC_CC, parameterName=PC_CC_offset}
  n16 = Saturate(None=Divide(A=Subtract(A=@n14, B=@n15), B=ScalarParameter() {defaultValue=500, group=PC_CC, parameterName=PC_CC_Radius}))
  n17 = ScalarParameter() {defaultValue=1, group=PC_CC, parameterName=PC_CC_Brightness1}
  n18 = ScalarParameter() {defaultValue=1, group=PC_CC, parameterName=PC_CC_Brightness2}
  n19 = LinearInterpolate(
   A=ScalarParameter() {defaultValue=1, group=PC_CC, parameterName=PC_CC_Contrast1}
   B=ScalarParameter() {defaultValue=1, group=PC_CC, parameterName=PC_CC_Contrast2}
   Alpha=@n16
  )
  n20 = Multiply(
   A=LinearInterpolate(
    A=VectorParameter() {group=PC_CC, parameterName=PC_CC_Color1, defaultValue=<rgba 1 1 1 1>}.RGB
    B=VectorParameter() {group=PC_CC, parameterName=PC_CC_Color2, defaultValue=<rgba 1 1 1 1>}.R
    Alpha=@n16
   )
   B=Multiply(A=LinearInterpolate(A=@n17, B=@n18, Alpha=@n16), B=Power(Base=@n13, Exp=@n19))
  )
  n21 = LinearInterpolate(A=ScalarParameter() {group=PC_CC, parameterName=PC_CC_Desat1}, B=ScalarParameter() {group=PC_CC, parameterName=PC_CC_Desat2})
  n22 = Divide(
   A=Subtract(A=@n14, B=ScalarParameter() {defaultValue=500, group=PC_CC, parameterName=PC_CC_FadeIn_offset})
   B=ScalarParameter() {defaultValue=500, group=PC_CC, parameterName=PC_CC_FadeIn_Radius}
  )
  n23 = StaticSwitchParameter(True=LinearInterpolate(A=Desaturation(None=@n20, Fraction=@n21), B=@n13, Alpha=Saturate(None=@n22)), False=@n13) {parameterName=UseBP13_ColorCorrectionPC}
  n24 = ScalarParameter() {defaultValue=1.5, group=FixErrors, parameterName=MaxBirghtnesClamp}
  n25 = Clamp(None=@n13, Max=@n24)
  n26 = ScalarParameter() {defaultValue=0.5095, group=ColorCorrection, parameterName=MobileDesat2}
  n27 = Power(
   Base=Multiply(A=Desaturation(None=@n25, Fraction=@n26), B=ScalarParameter() {defaultValue=0.4872, group=ColorCorrection, parameterName=MobileBrightness})
   Exp=ScalarParameter() {defaultValue=1.0607, group=ColorCorrection, parameterName=MobileContrat}
  )
  n28 = Power(
   Base=SphereMask(
    A=ScreenPosition().ViewportUV
    B=Constant2Vector() {g=0.5, r=0.5}
    Radius=ScalarParameter() {defaultValue=0.8715, group=Vignette, parameterName=Vignette}
    Hardness=ScalarParameter() {defaultValue=0.419, group=Vignette, parameterName=Vignette_Hardness}
   )
   Exp=ScalarParameter() {defaultValue=2.056, group=Vignette, parameterName=Vignette_Contrast}
  )
  n29 = Clamp(None=Multiply(A=@n28, B=@n27), Max=@n24)
  n30 = Clamp(None=Power(Base=@n25, Exp=ScalarParameter() {defaultValue=80, group=Light1, parameterName=Light1_Contrast}), Max=@n24)
  n31 = Distance(A=VectorParameter() {group=Light, parameterName=Position, defaultValue=<rgba -140016 40160 688 1>}.RGB, B=WorldPosition().XYZ)
  n32 = ScalarParameter() {defaultValue=974.0735, group=Light1, parameterName=Light1_Distance}
  n33 = OneMinus(None=Saturate(None=Divide(A=@n31, B=@n32)))
  n34 = ScalarParameter() {defaultValue=2.3798, group=Light1, parameterName=Light1_Falloff}
  n35 = Multiply(
   A=@n30
   B=Multiply(A=VectorParameter() {group=Light1, parameterName=Light1_Color, defaultValue=<rgba 0.3542 0.1737 0.0275 1>}.RGB, B=Power(Base=@n33, Exp=@n34))
  )
  n36 = VectorParameter() {group=Light, parameterName=Position2, defaultValue=<rgba -140016 40160 688 1>}
  n37 = WorldPosition()
  n38 = Divide(
   A=StaticSwitchParameter(True=Distance(A=@n36.RGB, B=@n37.XYZ), False=@n31) {parameterName=UseTwoLightPositions}
   B=ScalarParameter() {defaultValue=974.0735, group=Light2, parameterName=Light2_Distance2}
  )
  n39 = Power(Base=OneMinus(None=Saturate(None=@n38)), Exp=ScalarParameter() {defaultValue=2.3798, group=Light2, parameterName=Light2_Falloff2})
  n40 = Multiply(A=@n35, B=OneMinus(None=Saturate(None=@n39)))
  n41 = Add(
   A=@n27
   B=Multiply(
    A=StaticSwitchParameter(True=@n35, False=@n40) {parameterName=UseTwoLightPositions}
    B=ScalarParameter() {defaultValue=1.0607, group=Light1, parameterName=Light1_Brightness}
   )
  )
  n42 = VectorParameter() {group=Light2, parameterName=Light2_Color, defaultValue=<rgba 0.6838 0.6551 1 1>}
  n43 = Multiply(A=Multiply(A=@n30, B=Multiply(A=@n39, B=@n42.RGB)), B=ScalarParameter() {defaultValue=1.0607, group=Light2, parameterName=Light2_Brightness})
  n44 = Clamp(None=Multiply(A=@n28, B=Add(A=@n41, B=@n43)), Max=@n24)
  n45 = ComponentMask(None=ShadingPathSwitch(Default=@n23, Deferred=@n23, Forward=@n23, Mobile=StaticSwitchParameter(True=@n29, False=@n44) {parameterName=Neutral})) {b=true, g=true, r=true}
  n46 = SceneTexture() {sceneTextureId=PPI_CustomDepth}
  n47 = Subtract(A=@n46.Color, B=Add(A=SceneTexture() {sceneTextureId=PPI_SceneDepth}.Color, B=ScalarParameter() {defaultValue=3, parameterName=Bias}))
  n48 = ComponentMask(None=OneMinus(None=Saturate(None=@n47))) {r=true}
  n49 = ComponentMask(None=ScreenPosition().ViewportUV) {g=true, r=true}
  n50 = Subtract(A=Divide(A=ViewSize(), B=Constant() {r=320}), B=Constant() {r=2}) {constB=0}
  n51 = Multiply(A=Clamp(None=Ceil(None=@n50)) {maxDefault=2, minDefault=1}, B=SceneTexelSize()) {constA=2}
  n52 = Constant()
  n53 = AppendVector(A=ComponentMask(None=@n51) {desc=x, r=true}, B=@n52)
  n54 = Add(A=@n49, B=@n53) {desc=x+}
  n55 = SceneTexture(UVs=@n54) {sceneTextureId=PPI_CustomDepth}
  n56 = Saturate(None=Subtract(A=@n55.Color, B=SceneTexture(UVs=@n54) {sceneTextureId=PPI_SceneDepth}.Color))
  n57 = AppendVector(A=@n52, B=ComponentMask(None=@n51) {desc=y, g=true}) {desc=y+}
  n58 = Add(A=@n49, B=@n57) {desc=y+}
  n59 = SceneTexture(UVs=@n58) {sceneTextureId=PPI_CustomDepth}
  n60 = Saturate(None=Subtract(A=@n59.Color, B=SceneTexture(UVs=@n58) {sceneTextureId=PPI_SceneDepth}.Color))
  n61 = Constant() {r=-1}
  n62 = Add(A=@n49, B=Multiply(A=@n53, B=@n61)) {desc=x-}
  n63 = SceneTexture(UVs=@n62) {sceneTextureId=PPI_CustomDepth}
  n64 = Saturate(None=Subtract(A=@n63.Color, B=SceneTexture(UVs=@n62) {sceneTextureId=PPI_SceneDepth}.Color))
  n65 = Add(A=@n49, B=Multiply(A=@n61, B=@n57)) {desc=y-}
  n66 = SceneTexture(UVs=@n65) {sceneTextureId=PPI_CustomDepth}
  n67 = Saturate(None=Subtract(A=@n66.Color, B=SceneTexture(UVs=@n65) {sceneTextureId=PPI_SceneDepth}.Color))
  n68 = MaterialFunctionCall!(
   Base=Multiply(A=@n48, B=@n46.Color)
   BaseSampleBias
   Scale
   Sample1=Multiply(A=OneMinus(None=@n56), B=@n55.Color)
   Sample2=Multiply(A=OneMinus(None=@n60), B=@n59.Color)
   Sample3=Multiply(A=OneMinus(None=@n64), B=@n63.Color)
   Sample4=Multiply(A=OneMinus(None=@n67), B=@n66.Color)
   Divisor=Constant() {r=-16}
   UseCenteredLines
   UseRGB
  ) {materialFunction=/Game/Packages/jl_tests/combine4Samples}
  n69 = Saturate(None=Power(Base=ComponentMask(None=@n68.Result) {r=true}, Exp=ScalarParameter() {defaultValue=80, parameterName=Contrast}))
  n70 = Constant() {desc=Stencil ID, r=20}
  n71 = Constant()
  n72 = If(
   A=ComponentMask(None=SceneTexture(UVs=@n54) {sceneTextureId=PPI_CustomStencil}.Color) {r=true}
   B=@n70
   "A > B"=@n71
   "A == B"=Constant() {r=1}
   "A < B"=@n71
  )
  n73 = SceneTexture(UVs=@n58) {sceneTextureId=PPI_CustomStencil}
  n74 = Constant()
  n75 = ComponentMask(None=SceneTexture(UVs=@n62) {sceneTextureId=PPI_CustomStencil}.Color) {r=true}
  n76 = Constant()
  n77 = Constant() {r=1}
  n78 = ComponentMask(None=SceneTexture(UVs=@n65) {sceneTextureId=PPI_CustomStencil}.Color) {r=true}
  n79 = Constant()
  n80 = Constant() {r=1}
  n81 = Add(
   A=If(A=ComponentMask(None=@n73.Color) {r=true}, B=@n70, "A > B"=@n74, "A == B"=Constant() {r=1}, "A < B"=@n74)
   B=Add(A=If(A=@n75, B=@n70, "A > B"=@n76, "A == B"=@n77, "A < B"=@n76), B=If(A=@n78, B=@n70, "A > B"=@n79, "A == B"=@n80, "A < B"=@n79))
  )
  n82 = OneMinus(None=Saturate(None=Add(A=@n72, B=@n81)))
  n83 = Panner(
   Coordinate=Multiply(A=@n11.ViewportUV, B=ScalarParameter() {defaultValue=2, parameterName=Tiling})
   Speed=AppendVector(A=ScalarParameter() {defaultValue=0.4, parameterName=Pan}, B=Constant())
  )
  n84 = ScalarParameter() {defaultValue=1.7003, parameterName=rot}
  n85 = Power(
   Base=TextureSample(UVs=Rotator(Coordinate=@n83, Time=@n84)) {texture=/Game/Global/Textures/Default/Masks_and_Gradients/WeaponGradients/T_Gradient_DesertSky_Desaturated}.R
   Exp=ScalarParameter() {defaultValue=2.7082, parameterName=ShimmerContrast}
  )
  n86 = ScalarParameter() {defaultValue=0.4, parameterName=Shimmer}
  n87 = ScalarParameter() {defaultValue=0.2, parameterName=Speed}
  n88 = LinearInterpolate(
   A=ScalarParameter() {defaultValue=-0.6762, parameterName=c1}
   B=ScalarParameter() {defaultValue=1, parameterName=c2}
   Alpha=Sine(None=Multiply(A=@n1, B=@n87))
  )
  n89 = Clamp(None=Multiply(A=Multiply(A=@n85, B=@n86), B=Saturate(None=@n88)))
  n90 = VectorParameter() {parameterName=FinalTint, defaultValue=<rgba 1 1 1 1>}
  n91 = StaticSwitchParameter(True=@n45, False=Add(A=Multiply(A=@n89, B=@n48), B=Multiply(A=@n90.RGB, B=@n45))) {parameterName=Neutral}
  n92 = FeatureLevelSwitch(
   Default=LinearInterpolate(A=@n45, B=VectorParameter() {parameterName=OutlineColor, defaultValue=<rgba 1 0.896 0.725 1>}.RGB, Alpha=Multiply(A=@n69, B=@n82))
   ES2=@n91
   ES3_1=@n91
  )
  n93 = Desaturation(None=@n92)
  n94 = LinearInterpolate(A=@n10, B=@n6, Alpha=@n93)
  LinearInterpolate(A=@n94, B=Constant() {r=1}, Alpha=@n9)
  n95 = Desaturation(None=@n25, Fraction=ScalarParameter() {defaultValue=0.2667, group=SceneColor, parameterName=Scene Desat})
  n96 = ScalarParameter() {defaultValue=0.4624, group=SceneColor, parameterName=Scene Contrast}
  Multiply(
   A=Clamp(None=Power(Base=@n95, Exp=@n96), Max=@n24)
   B=VectorParameter() {group=SceneColor, parameterName=Scene Color, defaultValue=<rgba 50 1.7729 0 1>}.RGB
  )
  OneMinus(None=Multiply(A=OneMinus(None=@n9), B=@n94))
  OneMinus(None=@n93)
  Saturate(None=Ceil(None=ComponentMask(None=@n46.Color) {r=true}))
  ScalarParameter() {defaultValue=3, parameterName=NumberOfPixels}
  SceneTexture() {sceneTextureId=PPI_PostProcessInput0}
  Sine(None=ComponentMask(None=Rotator(Coordinate=@n12, Time=@n84)) {r=true})
  n97 = Add(A=@n9, B=@n10)
  StaticSwitchParameter(True=Multiply(A=Saturate(None=@n97), B=@n92), False=@n92) {parameterName=UseUIShadow}
 ]
}
